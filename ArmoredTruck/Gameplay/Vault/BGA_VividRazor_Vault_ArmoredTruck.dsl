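; Armored truck vault door Blueprint that shows thermite and weak-point effects and spawns convoy loot when opened.
; A replicated alert flag drives thermite materials, Niagara and audio, while each weak point's socket-name suffix selects the glowing tumbler.
; On opening, the door hides, then the authority builds a loot table name from the parent's convoy gameplay tag and playlist context tags.
; It spawns that loot at LootLocation and destroys the vault shortly afterward.
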(event EventBeginPlay
  (|CallParentFunction))

(event Collision|EventActorBeginOverlap
  (params (OtherActor :type Actor%20Object%20Reference))
  (|CallParentFunction OtherActor) ; This node is disabled and will not be called.\nDrag off pins to build functionality.
)

(event EventTick
  (params (DeltaSeconds :type Float%20%28single-precision%29))
  (|CallParentFunction DeltaSeconds) ; This node is disabled and will not be called.\nDrag off pins to build functionality.
)

(event Custom|DestroyVault
  (Utilities|FlowControl|Delay 0.5)
  (Actor|DestroyActor))

(event EventHandleWeakPointSpawned
  (params (PointConfig :type Vault%20Weak%20Point%20Config%20Structure%20%28by%20ref%29))
  (bind (_shapename _socketname) (Utilities|Struct|BreakVaultWeakPointConfig PointConfig))
  (|CallParentFunction PointConfig)
  (Niagara|SetNiagaraVariable(Int32) (Variables|Default|GetNS_VividRazor_Weakpoint_Glow) "CurrentTumbler" (Utilities|String|StringToInteger (Utilities|String|Right (Utilities|String|ToString(Name) _socketname) 1))))

(event EventResetVisuals
  (bind _rootmesh (Variables|VividRazorVaultDoor|GetRootMesh))
  (|CallParentFunction)
  (Components|StaticMesh|SetStaticMesh (Variables|VividRazorVaultDoor|GetRootMesh) "/ArmoredTruck/Assets/Meshes/ArmoredTruck_Door.ArmoredTruck_Door")
  (Rendering|Material|SetMaterial _rootmesh 0 "/ArmoredTruck/Assets/Materials/M_FN_Valet_ArmoredTruck.M_FN_Valet_ArmoredTruck")
  (Rendering|Material|SetMaterial _rootmesh 1 "/ArmoredTruck/Assets/Materials/M_FN_Valet_ArmoredTruck_Tumblers.M_FN_Valet_ArmoredTruck_Tumblers"))

(event EventHandleAlertStarted
  (|CallParentFunction)
  (Networking|SwitchHasAuthority
    (:Authority
      (Variables|Default|SetbAlertFxStarted true))))

(fn DetermineWhichConvoy
  (bind _returnvalue (GameplayTags|HasMatchingGameplayTag(Message) (Actor|GetAttachParentActor) "(TagName=\"Athena.AI.Lager.SpawnAction.Vehicle.BlueTeam.01\")"))
  (if _returnvalue
    (Variables|Default|SetConvoyName "Loot_DragonPop_Convoy_A_")
    (else
      (bind _returnvalue_2 (GameplayTags|HasMatchingGameplayTag(Message) (Actor|GetAttachParentActor) "(TagName=\"Athena.AI.Lager.SpawnAction.Vehicle.BlueTeam.02\")"))
      (if _returnvalue_2
        (Variables|Default|SetConvoyName "Loot_DragonPop_Convoy_B_")
        (else
          (bind _returnvalue_3 (GameplayTags|HasMatchingGameplayTag(Message) (Actor|GetAttachParentActor) "(TagName=\"Athena.AI.Lager.SpawnAction.Vehicle.BlueTeam.03\")"))
          (if _returnvalue_3
            (Variables|Default|SetConvoyName "Loot_DragonPop_Convoy_C_")
            (else
              (Variables|Default|SetConvoyName "Loot_DragonPop_Convoy_A_"))))))))

(fn MakeGameplayCueDoorOpenParams
  (return (Ability|GameplayCue|MakeGameplayCueParameters :NormalizedMagnitude 0.0 :RawMagnitude 0.0 :Location (Transformation|GetActorLocation) :Normal (* (Transformation|GetActorForwardVector) -1) :GameplayEffectLevel 1 :AbilityLevel 1 :bReplicateLocationWhenUsingMinimalRepProxy false)))

(fn OnDoorOpen
  (|CallParentFunction)
  (Rendering|SetVisibility (Variables|VividRazorVaultDoor|GetRootMesh) false true)
  ; We may want to test the location to make sure it's okay to spawn a bit off the back off the truck, otherwise we can bring it in closer.
  (Networking|SwitchHasAuthority
    (:Authority
      (CallFunction|DetermineWhichConvoy)
      (Class|AthenaFunctionLibrary|SpawnLootfromTable (CallFunction|PickLootbyPlaylist) 1 -1 (Transformation|GetWorldLocation (Variables|Default|GetLootLocation)) "0.000000,0.000000,90.000000" 0 true true false 0 "Container" "Chest")
      (CallFunction|DestroyVault))))

(fn OnRep_bAlertFxStarted
  (Networking|CanExecuteCosmeticEvents
    (:True
      (if (Variables|Default|GetbAlertFxStarted)
        (Visuals|UpdateThermiteMaterials true)
        (Components|Activation|Activate (Variables|Default|GetNS_Thermite_for_Vault_Proto) false)
        (CallFunction|ThermiteAudio true)))))

(fn PickLootByPlaylist
  (bind _keys (Utilities|Map|Keys (Variables|Default|GetLootGroupByPlaylistMap)))
  (bind (_array_element _array_index) (Utilities|Array|ForEachLoopwithBreak _keys)
    (:LoopBody
      (bind _returnvalue (Fortnite|GameplayTags|HasTagfromName (GameplayTags|GetTagName _array_element) (Class|FortGameStateAthena|GetAthenaPlaylistContextTags (Fortnite|Athena|GameState|GetGameStateAthena))))
      (if _returnvalue
        (bind _values (Utilities|Map|Values (Variables|Default|GetLootGroupByPlaylistMap)))
        (Variables|Default|SetVaultLootGroupName (Utilities|Array|Get(aref) _values _array_index)))
      (Variables|Default|SetVaultLootGroupName "Squad") ; Default
    )
    (:Completed
      _))
  (return (Utilities|String|StringToName (Utilities|String|Append (Utilities|String|ToString(Name) (Variables|Default|GetConvoyName)) (Utilities|String|ToString(Name) (Variables|Default|GetVaultLootGroupName))))))

(fn ConstructionScript
  (|CallParentFunction))
